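# Level Sequence for the Meowscles (BuffCat) intro cinematic in the Season 12 Battle Pass frontend menu, about 16 seconds at 30 fps.
# A camera cut track steps through six cine camera shots with keyed focus and quick repositions, alongside an animated frontend camera.
# The mannequin plays per-shot animations synced with barbell gadgets toggled by visibility.
# Set props shift and a nav-object mannequin hides during playback, framed by fades and one audio section.
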LevelSequence /MiscReferencersOfCosmetics/Animation/Game/MainPlayer/Menu/BR/S12_BP/Meowscles/Meowscles_Intro
bindings [
 CameraComponent {
  id=#1
  parent_id=#2
  possessed_class=CineCameraComponent
  tracks [
   {
    type=MovieSceneFloatTrack
    display_name=Current Focal Length
    sections [{type=MovieSceneFloatSection, channels [None {type=MovieSceneScriptingFloatChannel, const [20,27]}]}]
   }
   {
    type=MovieSceneFloatTrack
    display_name="Manual Focus Distance (Focus Settings)"
    sections [{type=MovieSceneFloatSection, channels [None {type=MovieSceneScriptingFloatChannel, default=174.3223, const [174.3223,1]}]}]
   }
   {
    type=MovieSceneFloatTrack
    display_name=Current Aperture
    sections [{type=MovieSceneFloatSection, channels [None {type=MovieSceneScriptingFloatChannel, default=2.8}]}]
   }
  ]
 }
 Buff_Cat_LargeBarbell_Gadget {
  id=#3
  possessed_class=SkeletalMeshActor
  tracks [
   {
    type=MovieScene3DTransformTrack
    sections [{
      type=MovieScene3DTransformSection
      channels [
       Location.X {type=MovieSceneScriptingDoubleChannel, default=2304}
       Location.Y {type=MovieSceneScriptingDoubleChannel, default=16910}
       Location.Z {type=MovieSceneScriptingDoubleChannel, default=678}
       Rotation.X {type=MovieSceneScriptingDoubleChannel, default=0}
       Rotation.Y {type=MovieSceneScriptingDoubleChannel, default=0}
       Rotation.Z {type=MovieSceneScriptingDoubleChannel, default=90}
       Scale.X {type=MovieSceneScriptingDoubleChannel, default=1}
       Scale.Y {type=MovieSceneScriptingDoubleChannel, default=1}
       Scale.Z {type=MovieSceneScriptingDoubleChannel, default=1}
      ]
     }]
   }
   {
    type=MovieSceneSkeletalAnimationTrack
    sections [
     {type=MovieSceneSkeletalAnimationSection, end_frame=138, start_frame=111}
     {type=MovieSceneSkeletalAnimationSection, end_frame=162, start_frame=138}
    ]
   }
   {
    type=MovieSceneVisibilityTrack
    sections [{
      type=MovieSceneVisibilitySection
      channels [None {type=MovieSceneScriptingBoolChannel, default=true, keys [[0,false],[111.0188,true],[135.4712,true],[162.0013,false]]}]
     }]
   }
  ]
 }
 BP_Frontend_EventLevel_NavObject_Mannequin2 {
  id=#4
  possessed_class=BP_Frontend_EventLevel_NavObject_Mannequin_C
  tracks [{
    type=MovieSceneVisibilityTrack
    sections [{
      type=MovieSceneVisibilitySection
      channels [None {type=MovieSceneScriptingBoolChannel, default=true, keys [[-57.96,true],[0.3425,false],[469.805,true]]}]
     }]
   }]
 }
 Shot_6000 {
  id=#5
  possessed_class=CineCameraActor
  tracks [
   {
    type=MovieScene3DTransformTrack
    sections [{
      type=MovieScene3DTransformSection
      channels [
       Location.X {
        type=MovieSceneScriptingDoubleChannel
        default=2428.2681
        thinned_from=253
        keys [[216.9987,2428.2681],[235.9987,2428.2681],[236.9987,2472.0977],[238.9987,2371.8357],[240.9987,2389.7397],[307.9987,2386.4841],[308.9987,2390.3723],[309.9987,2401.8247],[310.9987,2422.9766],[311.9987,2453.7485],[314.9987,2564.7017],[315.9987,2590.0518],[316.9987,2604.3052],[317.9987,2608.4836],[321.9987,2592.6235],[468.9987,2592.6235]]
       }
       Location.Y {
        type=MovieSceneScriptingDoubleChannel
        default=16868.4082
        thinned_from=253
        keys [[216.9987,16868.4082],[312.9987,16831.457],[320.9987,17075],[468.9987,17079.8027]]
       }
       Location.Z {
        type=MovieSceneScriptingDoubleChannel
        default=1055.4435
        thinned_from=253
        keys [[216.9987,1055.4435],[235.9987,1055.4435],[236.9987,1082.7837],[238.9987,1020.242],[240.9987,1031.4102],[304.9987,1031.4102],[305.9987,1028.9875],[306.9987,1022.1298],[307.9987,1011.4649],[308.9987,997.6276],[310.9987,962.9355],[315.9987,862.6216],[316.9987,844.6659],[317.9987,828.6901],[318.9987,815.2574],[319.9987,804.9422],[320.9987,798.3261],[321.9987,795.992],[468.9987,795.992]]
       }
       Rotation.X {
        type=MovieSceneScriptingDoubleChannel
        default=0
        thinned_from=253
        keys [[216.9987,0],[304.9987,0],[305.9987,0.0078],[306.9987,0.0296],[307.9987,0.0627],[308.9987,0.1044],[309.9987,0.1521],[310.9987,0.2041],[311.9987,0.2586],[314.9987,0.4277],[315.9987,0.4827],[316.9987,0.5352],[317.9987,0.5834],[318.9987,0.6254],[319.9987,0.6586],[320.9987,0.6804],[321.9987,0.6882],[468.9987,0.6882]]
       }
       Rotation.Y {
        type=MovieSceneScriptingDoubleChannel
        default=-30.7749
        thinned_from=253
        keys [[216.9987,-30.7749],[304.9987,-30.7749],[305.9987,-30.3589],[306.9987,-29.1787],[307.9987,-27.3364],[308.9987,-24.9337],[309.9987,-22.0725],[310.9987,-18.8548],[311.9987,-15.3823],[312.9987,-11.7571],[313.9987,-8.0809],[314.9987,-4.4557],[315.9987,-0.9834],[316.9987,2.2341],[317.9987,5.0951],[318.9987,7.4976],[319.9987,9.3397],[320.9987,10.5197],[321.9987,10.9357],[468.9987,10.9357]]
       }
       Rotation.Z {
        type=MovieSceneScriptingDoubleChannel
        default=162.688
        thinned_from=253
        keys [[216.9987,162.688],[304.9987,162.688],[305.9987,165.0286],[306.9987,171.6688],[307.9987,182.0372],[308.9987,195.5626],[309.9987,211.6737],[310.9987,229.7988],[311.9987,249.3654],[314.9987,310.9794],[315.9987,330.5736],[316.9987,348.7366],[317.9987,364.8918],[318.9987,378.4622],[319.9987,388.8702],[320.9987,395.5383],[321.9987,397.8891],[468.9987,397.8891]]
       }
       Scale.X {type=MovieSceneScriptingDoubleChannel, default=1, const [1,253]}
       Scale.Y {type=MovieSceneScriptingDoubleChannel, default=1, const [1,253]}
       Scale.Z {type=MovieSceneScriptingDoubleChannel, default=1, const [1,253]}
      ]
     }]
   }
   {type=MovieSceneSpawnTrack, sections [{type=MovieSceneSpawnSection, channels [None {type=MovieSceneScriptingBoolChannel, default=true}]}]}
  ]
 }
 Shot_5000 {
  id=#6
  possessed_class=CineCameraActor
  tracks [
   {
    type=MovieScene3DTransformTrack
    sections [{
      type=MovieScene3DTransformSection
      channels [
       Location.X {
        type=MovieSceneScriptingDoubleChannel
        default=1787.996
        thinned_from=56
        keys [[162.0175,1787.996],[183.0175,1787.996],[184.0175,1777.4794],[185.0175,1757.9482],[186.0175,1747.4316],[189.0175,1756.5356],[217.0175,1756.5356]]
       }
       Location.Y {
        type=MovieSceneScriptingDoubleChannel
        default=17261.1836
        thinned_from=56
        keys [[162.0175,17261.1836],[183.0175,17261.1836],[186.0175,17348.752],[217.0175,17329.0332]]
       }
       Location.Z {
        type=MovieSceneScriptingDoubleChannel
        default=1010.9452
        thinned_from=56
        keys [[162.0175,1010.9452],[183.0175,1010.9452],[184.0175,1014.18],[185.0175,1020.1874],[186.0175,1023.4221],[217.0175,1025.6049]]
       }
       Rotation.X {type=MovieSceneScriptingDoubleChannel, default=0.3099, const [0.3099,56]}
       Rotation.Y {type=MovieSceneScriptingDoubleChannel, default=-5.7393, const [-5.7393,56]}
       Rotation.Z {type=MovieSceneScriptingDoubleChannel, default=114.784, const [114.7841,56]}
       Scale.X {type=MovieSceneScriptingDoubleChannel, default=1, const [1,56]}
       Scale.Y {type=MovieSceneScriptingDoubleChannel, default=1, const [1,56]}
       Scale.Z {type=MovieSceneScriptingDoubleChannel, default=1, const [1,56]}
      ]
     }]
   }
   {type=MovieSceneSpawnTrack, sections [{type=MovieSceneSpawnSection, channels [None {type=MovieSceneScriptingBoolChannel, default=true}]}]}
  ]
 }
 CameraComponent {
  id=#7
  parent_id=#8
  possessed_class=CineCameraComponent
  tracks [
   {
    type=MovieSceneFloatTrack
    display_name=Current Focal Length
    sections [{type=MovieSceneFloatSection, channels [None {type=MovieSceneScriptingFloatChannel, const [35,53]}]}]
   }
   {
    type=MovieSceneFloatTrack
    display_name="Manual Focus Distance (Focus Settings)"
    sections [{type=MovieSceneFloatSection, channels [None {type=MovieSceneScriptingFloatChannel, default=5, const [163.1648,1]}]}]
   }
   {
    type=MovieSceneFloatTrack
    display_name=Current Aperture
    sections [{type=MovieSceneFloatSection, channels [None {type=MovieSceneScriptingFloatChannel, default=2.8}]}]
   }
  ]
 }
 S_Yacht_Rug_01 {
  id=#9
  possessed_class=FortStaticMeshActor
  tracks [{
    type=MovieScene3DTransformTrack
    sections [{
      type=MovieScene3DTransformSection
      channels [
       Location.X {type=MovieSceneScriptingDoubleChannel, default=2517.2966, const [2517.2966,3]}
       Location.Y {type=MovieSceneScriptingDoubleChannel, default=21787.502, keys [[302.5462,16992.6055],[338.8062,22000],[483.605,22000],[502.6812,17000]]}
       Location.Z {type=MovieSceneScriptingDoubleChannel, default=780, const [780,3]}
       Rotation.X {type=MovieSceneScriptingDoubleChannel, default=0}
       Rotation.Y {type=MovieSceneScriptingDoubleChannel, default=0}
       Rotation.Z {type=MovieSceneScriptingDoubleChannel, default=0}
       Scale.X {type=MovieSceneScriptingDoubleChannel, default=1}
       Scale.Y {type=MovieSceneScriptingDoubleChannel, default=1}
       Scale.Z {type=MovieSceneScriptingDoubleChannel, default=1.5}
      ]
     }]
   }]
 }
 CameraComponent {
  id=#10
  parent_id=#5
  possessed_class=CineCameraComponent
  tracks [
   {
    type=MovieSceneFloatTrack
    display_name=Current Focal Length
    sections [{type=MovieSceneFloatSection, channels [None {type=MovieSceneScriptingFloatChannel, const [20,253]}]}]
   }
   {
    type=MovieSceneFloatTrack
    display_name="Manual Focus Distance (Focus Settings)"
    sections [{type=MovieSceneFloatSection, channels [None {type=MovieSceneScriptingFloatChannel, default=768.4797, const [768.4797,1]}]}]
   }
   {
    type=MovieSceneFloatTrack
    display_name=Current Aperture
    sections [{type=MovieSceneFloatSection, channels [None {type=MovieSceneScriptingFloatChannel, default=2.8}]}]
   }
  ]
 }
 Shot_3000 {
  id=#2
  possessed_class=CineCameraActor
  tracks [
   {
    type=MovieScene3DTransformTrack
    sections [{
      type=MovieScene3DTransformSection
      channels [
       Location.X {type=MovieSceneScriptingDoubleChannel, default=2486.6082, const [2486.6082,27]}
       Location.Y {type=MovieSceneScriptingDoubleChannel, default=17205.1758, const [17205.1758,27]}
       Location.Z {type=MovieSceneScriptingDoubleChannel, default=839.5886, const [839.5886,27]}
       Rotation.X {type=MovieSceneScriptingDoubleChannel, default=-1.3511, const [-1.3511,27]}
       Rotation.Y {type=MovieSceneScriptingDoubleChannel, default=20.8693, const [20.8693,27]}
       Rotation.Z {type=MovieSceneScriptingDoubleChannel, default=111.625, const [111.625,27]}
       Scale.X {type=MovieSceneScriptingDoubleChannel, default=1, const [1,27]}
       Scale.Y {type=MovieSceneScriptingDoubleChannel, default=1, const [1,27]}
       Scale.Z {type=MovieSceneScriptingDoubleChannel, default=1, const [1,27]}
      ]
     }]
   }
   {type=MovieSceneSpawnTrack, sections [{type=MovieSceneSpawnSection, channels [None {type=MovieSceneScriptingBoolChannel, default=true}]}]}
  ]
 }
 CameraComponent {
  id=#11
  parent_id=#12
  possessed_class=CineCameraComponent
  tracks [
   {
    type=MovieSceneFloatTrack
    display_name=Current Focal Length
    sections [{type=MovieSceneFloatSection, channels [None {type=MovieSceneScriptingFloatChannel, default=15}]}]
   }
   {
    type=MovieSceneFloatTrack
    display_name="Manual Focus Distance (Focus Settings)"
    sections [{
      type=MovieSceneFloatSection
      channels [None {type=MovieSceneScriptingFloatChannel, default=65.2378, keys [[138.02,221.675],[145.9962,220],[153.4412,65.2378]]}]
     }]
   }
   {
    type=MovieSceneFloatTrack
    display_name=Current Aperture
    sections [{type=MovieSceneFloatSection, channels [None {type=MovieSceneScriptingFloatChannel, default=2.8}]}]
   }
   {
    type=MovieScene3DTransformTrack
    sections [{
      type=MovieScene3DTransformSection
      channels [
       Location.X {type=MovieSceneScriptingDoubleChannel, default=0, const [0,2]}
       Location.Y {type=MovieSceneScriptingDoubleChannel, default=0, const [0,2]}
       Location.Z {type=MovieSceneScriptingDoubleChannel, default=0, const [0,2]}
       Rotation.X {type=MovieSceneScriptingDoubleChannel, default=0, const [0,2]}
       Rotation.Y {type=MovieSceneScriptingDoubleChannel, default=0, const [0,2]}
       Rotation.Z {type=MovieSceneScriptingDoubleChannel, default=0, const [0,2]}
       Scale.X {type=MovieSceneScriptingDoubleChannel, default=1, const [1,2]}
       Scale.Y {type=MovieSceneScriptingDoubleChannel, default=1, const [1,2]}
       Scale.Z {type=MovieSceneScriptingDoubleChannel, default=1, const [1,2]}
      ]
     }]
   }
  ]
 }
 Shot_2000 {
  id=#8
  possessed_class=CineCameraActor
  tracks [
   {
    type=MovieScene3DTransformTrack
    sections [{
      type=MovieScene3DTransformSection
      channels [
       Location.X {
        type=MovieSceneScriptingDoubleChannel
        default=2067.0376
        thinned_from=53
        keys [[59.0013,2067.0376],[84.0013,2067.0376],[85.0013,2062.2053],[87.0013,2046.4825],[111.0013,2053.428]]
       }
       Location.Y {
        type=MovieSceneScriptingDoubleChannel
        default=17391.1855
        thinned_from=53
        keys [[59.0013,17391.1855],[85.0013,17407.707],[87.0013,17454.2949],[111.0013,17433.6016]]
       }
       Location.Z {type=MovieSceneScriptingDoubleChannel, default=1035.0107, thinned_from=53, keys [[59.0013,1035.0107],[111.0013,1034.9978]]}
       Rotation.X {type=MovieSceneScriptingDoubleChannel, default=2.5591, thinned_from=53, keys [[59.0013,2.5591],[111.0013,2.5623]]}
       Rotation.Y {
        type=MovieSceneScriptingDoubleChannel
        default=0.1289
        thinned_from=53
        keys [[59.0013,0.1289],[84.0013,0.1289],[85.0013,0.0919],[86.0013,0.0232],[87.0013,-0.0138],[111.0013,-0.0138]]
       }
       Rotation.Z {
        type=MovieSceneScriptingDoubleChannel
        default=106.4861
        thinned_from=53
        keys [[59.0013,106.4861],[84.0013,106.4861],[85.0013,107.0223],[86.0013,108.0181],[87.0013,108.5542],[111.0013,108.5542]]
       }
       Scale.X {type=MovieSceneScriptingDoubleChannel, default=1, const [1,53]}
       Scale.Y {type=MovieSceneScriptingDoubleChannel, default=1, const [1,53]}
       Scale.Z {type=MovieSceneScriptingDoubleChannel, default=1, const [1,53]}
      ]
     }]
   }
   {type=MovieSceneSpawnTrack, sections [{type=MovieSceneSpawnSection, channels [None {type=MovieSceneScriptingBoolChannel, default=true}]}]}
  ]
 }
 Buff_Cat_SmallBarbell_Gadget {
  id=#13
  possessed_class=SkeletalMeshActor
  tracks [
   {
    type=MovieScene3DTransformTrack
    sections [{
      type=MovieScene3DTransformSection
      channels [
       Location.X {type=MovieSceneScriptingDoubleChannel, default=2304}
       Location.Y {type=MovieSceneScriptingDoubleChannel, default=16910}
       Location.Z {type=MovieSceneScriptingDoubleChannel, default=678}
       Rotation.X {type=MovieSceneScriptingDoubleChannel, default=0}
       Rotation.Y {type=MovieSceneScriptingDoubleChannel, default=0}
       Rotation.Z {type=MovieSceneScriptingDoubleChannel, default=90}
       Scale.X {type=MovieSceneScriptingDoubleChannel, default=1}
       Scale.Y {type=MovieSceneScriptingDoubleChannel, default=1}
       Scale.Z {type=MovieSceneScriptingDoubleChannel, default=1}
      ]
     }]
   }
   {type=MovieSceneSkeletalAnimationTrack, sections [{type=MovieSceneSkeletalAnimationSection, end_frame=59, start_frame=0}]}
   {type=MovieSceneSkeletalAnimationTrack, sections [{type=MovieSceneSkeletalAnimationSection, end_frame=216, start_frame=162}]}
   {
    type=MovieSceneVisibilityTrack
    sections [{
      type=MovieSceneVisibilitySection
      channels [None {type=MovieSceneScriptingBoolChannel, default=false, keys [[-0.005,false],[0.9975,true],[59.0088,false],[162,true],[216.9987,false]]}]
     }]
   }
  ]
 }
 Shot_1000 {
  id=#14
  possessed_class=CineCameraActor
  tracks [
   {
    type=MovieScene3DTransformTrack
    sections [{
      type=MovieScene3DTransformSection
      channels [
       Location.X {
        type=MovieSceneScriptingDoubleChannel
        default=1479.0787
        thinned_from=60
        keys [[0,1479.0787],[34,1479.0787],[35,1508.3954],[36,1562.6039],[37,1591.653],[38,1587.0387],[39,1578.4695],[40,1573.8553],[59,1573.8553]]
       }
       Location.Y {
        type=MovieSceneScriptingDoubleChannel
        default=16846.6855
        thinned_from=60
        keys [[0,16846.6855],[34,16846.6855],[37,16930.8848],[59,16914.5703]]
       }
       Location.Z {
        type=MovieSceneScriptingDoubleChannel
        default=1001.1243
        thinned_from=60
        keys [[0,1001.1243],[34,1001.1243],[35,996.6197],[36,986.9566],[37,981.058],[40,985.9312],[59,985.9312]]
       }
       Rotation.X {
        type=MovieSceneScriptingDoubleChannel
        default=-1.0071
        thinned_from=60
        keys [[0,-1.0071],[34,-1.0071],[35,-1.0092],[37,-1.0162],[59,-1.0162]]
       }
       Rotation.Y {
        type=MovieSceneScriptingDoubleChannel
        default=-10.3919
        thinned_from=60
        keys [[0,-10.3919],[34,-10.3919],[35,-11.0423],[36,-12.2503],[37,-12.9007],[59,-12.9007]]
       }
       Rotation.Z {type=MovieSceneScriptingDoubleChannel, default=33.1817, thinned_from=60, keys [[0,33.1817],[59,33.2269]]}
       Scale.X {type=MovieSceneScriptingDoubleChannel, default=1, const [1,60]}
       Scale.Y {type=MovieSceneScriptingDoubleChannel, default=1, const [1,60]}
       Scale.Z {type=MovieSceneScriptingDoubleChannel, default=1, const [1,60]}
      ]
     }]
   }
   {
    type=MovieSceneSpawnTrack
    sections [{type=MovieSceneSpawnSection, channels [None {type=MovieSceneScriptingBoolChannel, default=true, const [true,1]}]}]
   }
  ]
 }
 CameraComponent {
  id=#15
  parent_id=#6
  possessed_class=CineCameraComponent
  tracks [
   {
    type=MovieSceneFloatTrack
    display_name=Current Focal Length
    sections [{type=MovieSceneFloatSection, channels [None {type=MovieSceneScriptingFloatChannel, const [25,56]}]}]
   }
   {
    type=MovieSceneFloatTrack
    display_name="Manual Focus Distance (Focus Settings)"
    sections [{
      type=MovieSceneFloatSection
      channels [None {
        type=MovieSceneScriptingFloatChannel
        default=96.1782
        thinned_from=59
        keys [[0,5],[55,5],[162.0175,120.7266],[182.6313,120.7266],[186.8737,96.1782]]
       }]
     }]
   }
   {
    type=MovieSceneFloatTrack
    display_name=Current Aperture
    sections [{type=MovieSceneFloatSection, channels [None {type=MovieSceneScriptingFloatChannel, default=2.8}]}]
   }
  ]
 }
 Camera_Root_Movement {
  id=#16
  possessed_class=Frontend_EventLevel_Camera_C
  tracks [{
    type=MovieScene3DTransformTrack
    sections [{
      type=MovieScene3DTransformSection
      channels [
       Location.X {type=MovieSceneScriptingDoubleChannel, default=2710, keys [[0,637.1765],[5,3054.3296],[465,2755.2795]]}
       Location.Y {type=MovieSceneScriptingDoubleChannel, default=16441.2559, keys [[0,-1200.7053],[5,16211.2393],[465,16400.7832]]}
       Location.Z {type=MovieSceneScriptingDoubleChannel, default=975, keys [[0,551.1627],[5,1016.4581],[465,986.691]]}
       Rotation.X {type=MovieSceneScriptingDoubleChannel, default=0.0001, thinned_from=3, keys [[0,0],[465,0.0001]]}
       Rotation.Y {type=MovieSceneScriptingDoubleChannel, default=-1.6, keys [[0,0.5995],[5,0],[465,-1.6]]}
       Rotation.Z {type=MovieSceneScriptingDoubleChannel, default=-221.7053, keys [[0,131.592],[5,158.4375],[465,138.2947]]}
       Scale.X {type=MovieSceneScriptingDoubleChannel, default=1}
       Scale.Y {type=MovieSceneScriptingDoubleChannel, default=1}
       Scale.Z {type=MovieSceneScriptingDoubleChannel, default=1}
      ]
     }]
   }]
 }
 Shot_4000 {
  id=#12
  possessed_class=CineCameraActor
  tracks [
   {
    type=MovieScene3DTransformTrack
    sections [{
      type=MovieScene3DTransformSection
      channels [
       Location.X {type=MovieSceneScriptingDoubleChannel, default=2474.188, const [2474.188,2]}
       Location.Y {type=MovieSceneScriptingDoubleChannel, default=17301.1953, const [17301.1953,2]}
       Location.Z {type=MovieSceneScriptingDoubleChannel, default=1069.011, const [1069.011,2]}
       Rotation.X {type=MovieSceneScriptingDoubleChannel, default=-14.7645, const [-14.7645,26]}
       Rotation.Y {type=MovieSceneScriptingDoubleChannel, default=-51.6991, const [-51.6991,26]}
       Rotation.Z {type=MovieSceneScriptingDoubleChannel, default=120.9514, const [-239.0486,26]}
       Scale.X {type=MovieSceneScriptingDoubleChannel, default=1, const [1,26]}
       Scale.Y {type=MovieSceneScriptingDoubleChannel, default=1, const [1,26]}
       Scale.Z {type=MovieSceneScriptingDoubleChannel, default=1, const [1,26]}
      ]
     }]
   }
   {type=MovieSceneSpawnTrack, sections [{type=MovieSceneSpawnSection, channels [None {type=MovieSceneScriptingBoolChannel, default=true}]}]}
  ]
 }
 Prop_FreeWeights_6 {
  id=#17
  possessed_class=FortStaticMeshActor
  tracks [{
    type=MovieScene3DTransformTrack
    sections [{
      type=MovieScene3DTransformSection
      channels [
       Location.X {type=MovieSceneScriptingDoubleChannel, default=2836, thinned_from=5, keys [[220,2523.2769],[221,2836],[473,2836],[474,2523.2769]]}
       Location.Y {type=MovieSceneScriptingDoubleChannel, default=17032.7578, const [17032.7578,5]}
       Location.Z {type=MovieSceneScriptingDoubleChannel, default=780, const [780,5]}
       Rotation.X {type=MovieSceneScriptingDoubleChannel, default=0, const [0,5]}
       Rotation.Y {type=MovieSceneScriptingDoubleChannel, default=0, const [0,5]}
       Rotation.Z {type=MovieSceneScriptingDoubleChannel, default=-15.0001, const [-15.0001,5]}
       Scale.X {type=MovieSceneScriptingDoubleChannel, default=1.1, const [1.1,5]}
       Scale.Y {type=MovieSceneScriptingDoubleChannel, default=1.1, const [1.1,5]}
       Scale.Z {type=MovieSceneScriptingDoubleChannel, default=1.1, const [1.1,5]}
      ]
     }]
   }]
 }
 CameraComponent {
  id=#18
  parent_id=#14
  possessed_class=CineCameraComponent
  tracks [
   {
    type=MovieSceneFloatTrack
    display_name=Current Focal Length
    sections [{type=MovieSceneFloatSection, channels [None {type=MovieSceneScriptingFloatChannel, const [35,60]}]}]
   }
   {
    type=MovieSceneFloatTrack
    display_name="Manual Focus Distance (Focus Settings)"
    sections [{
      type=MovieSceneFloatSection
      channels [None {type=MovieSceneScriptingFloatChannel, default=122.45, keys [[0,5],[0.9962,191.3168],[33.455,191.3],[38.2362,87.5],[44.0738,125]]}]
     }]
   }
   {
    type=MovieSceneFloatTrack
    display_name=Current Aperture
    sections [{type=MovieSceneFloatSection, channels [None {type=MovieSceneScriptingFloatChannel, default=2.8}]}]
   }
  ]
 }
 BP12_BuffCat_Mannequin {
  id=#19
  possessed_class=SoldierMannequin_Male_C
  tracks [
   {
    type=MovieScene3DTransformTrack
    sections [{
      type=MovieScene3DTransformSection
      channels [
       Location.X {type=MovieSceneScriptingDoubleChannel, default=2304}
       Location.Y {type=MovieSceneScriptingDoubleChannel, default=16910}
       Location.Z {type=MovieSceneScriptingDoubleChannel, default=778}
       Rotation.X {type=MovieSceneScriptingDoubleChannel, default=0}
       Rotation.Y {type=MovieSceneScriptingDoubleChannel, default=0}
       Rotation.Z {type=MovieSceneScriptingDoubleChannel, default=0}
       Scale.X {type=MovieSceneScriptingDoubleChannel, default=1}
       Scale.Y {type=MovieSceneScriptingDoubleChannel, default=1}
       Scale.Z {type=MovieSceneScriptingDoubleChannel, default=1}
      ]
     }]
   }
   {type=MovieSceneSkeletalAnimationTrack, sections []}
   {
    type=MovieSceneSkeletalAnimationTrack
    sections [
     {type=MovieSceneSkeletalAnimationSection, end_frame=59, start_frame=0}
     {type=MovieSceneSkeletalAnimationSection, end_frame=111, start_frame=59}
     {type=MovieSceneSkeletalAnimationSection, end_frame=162, start_frame=138}
     {type=MovieSceneSkeletalAnimationSection, end_frame=216, start_frame=162}
     {type=MovieSceneSkeletalAnimationSection, end_frame=469, start_frame=216}
     {type=MovieSceneSkeletalAnimationSection, end_frame=138, start_frame=111}
    ]
   }
   {
    type=MovieSceneVisibilityTrack
    sections [{type=MovieSceneVisibilitySection, channels [None {type=MovieSceneScriptingBoolChannel, default=true}]}]
   }
  ]
 }
 CameraComponent {
  id=#20
  parent_id=#16
  possessed_class=CameraComponent
  tracks [{
    type=MovieSceneFloatTrack
    display_name=Field Of View
    sections [{type=MovieSceneFloatSection, channels [None {type=MovieSceneScriptingFloatChannel, default=80, thinned_from=3, keys [[0,75],[465,80]]}]}]
   }]
 }
 Buff_Cat_SmallBarbell_Gadget2 {
  id=#21
  possessed_class=SkeletalMeshActor
  tracks [
   {
    type=MovieScene3DTransformTrack
    sections [{
      type=MovieScene3DTransformSection
      channels [
       Location.X {type=MovieSceneScriptingDoubleChannel, default=2304}
       Location.Y {type=MovieSceneScriptingDoubleChannel, default=16910}
       Location.Z {type=MovieSceneScriptingDoubleChannel, default=678}
       Rotation.X {type=MovieSceneScriptingDoubleChannel, default=0}
       Rotation.Y {type=MovieSceneScriptingDoubleChannel, default=0}
       Rotation.Z {type=MovieSceneScriptingDoubleChannel, default=90}
       Scale.X {type=MovieSceneScriptingDoubleChannel, default=1}
       Scale.Y {type=MovieSceneScriptingDoubleChannel, default=1}
       Scale.Z {type=MovieSceneScriptingDoubleChannel, default=1}
      ]
     }]
   }
   {type=MovieSceneSkeletalAnimationTrack, sections [{type=MovieSceneSkeletalAnimationSection, end_frame=216, start_frame=162}]}
   {
    type=MovieSceneVisibilityTrack
    sections [{
      type=MovieSceneVisibilitySection
      channels [None {type=MovieSceneScriptingBoolChannel, default=false, keys [[0.7812,false],[161.9987,true],[216.9987,false]]}]
     }]
   }
  ]
 }
]
display_rate {denominator=1, numerator=30}
tick_resolution {denominator=1, numerator=24000}
tracks [
 {
  type=MovieSceneFadeTrack
  sections [{type=MovieSceneFadeSection, channels [None {type=MovieSceneScriptingFloatChannel, default=1, keys [[0,1],[5,0],[468.0025,0],[469.0013,1]]}]}]
 }
 {
  type=MovieSceneAudioTrack
  display_name=Audio
  sections [{
    type=MovieSceneAudioSection
    end_frame=472
    start_frame=2
    channels [
     Volume {type=MovieSceneScriptingFloatChannel, default=1}
     Pitch {type=MovieSceneScriptingFloatChannel, default=1}
     AttachActor {type=MovieSceneScriptingActorReferenceChannel, default="<Struct 'MovieSceneObjectBindingID' (0x8515dfdaa00) {}>"}
    ]
   }]
 }
 {
  type=MovieSceneCameraCutTrack
  sections [
   {type=MovieSceneCameraCutSection, end_frame=59, start_frame=0}
   {type=MovieSceneCameraCutSection, end_frame=110, start_frame=59}
   {type=MovieSceneCameraCutSection, end_frame=138, start_frame=110}
   {type=MovieSceneCameraCutSection, end_frame=162, start_frame=138}
   {type=MovieSceneCameraCutSection, end_frame=216, start_frame=162}
   {type=MovieSceneCameraCutSection, end_frame=468, start_frame=216}
  ]
 }
]
